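# Niagara idle effect for the Ice Cream pickaxe that gives off faint, cyan-tinted cold mist sprites that sink, grow and fade.
# MistIdle and MistIdle_Low gate their spawn rate on an owner-speed threshold, with Low falling more gently and offset in camera depth.
# MistTrail spawns by distance moved and inherits part of the velocity.
# All three use SubUV animation, drag and randomized dynamic material parameters.
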NiagaraSystem /BRCosmetics/Weapons/FORT_Melee/Pickaxe_IceCream/FX/NS_Pickaxe_IceCream_Idle
dependencies {
 usedDataInterfaces [
  &NiagaraDataInterfaceCurve
  &NiagaraDataInterfaceVector2DCurve
  &NiagaraDataInterfaceSpriteRendererInfo
 ]
 usedDynamicInputs [&Vector2DFromFloat]
 usedModules [&SystemState]
}
emitters {
 MistIdle {
  emitterUpdateScript [
   &EmitterState()
   &SpawnRate(
    SpawnRate=&MakeCustomFloatFromBool(
     Bool=&SetBoolByFloatComparison(A=&VectorLength(Vector=$Engine.Owner.Velocity), B=100, Comparison Type=::NewEnumerator6)
     True Float=8
     False Float=0
    )
    Use Spawn Probability=false
   )
  ]
  particleSpawnScript [
   /SetVariables(Particles.Acceleration:Vector3f=&RandomRangeVector(Minimum=<0 0 -32>, Maximum=<0 0 -40>, Evaluation Type=::NewEnumerator1)) {isSetParameters=true, module=SetVariables_#1}
   &InitializeParticle(
    Color Maximum=<rgba 1 1 1 0.02>
    Color Minimum=<rgba 0.5 0.5 0.5 0.02>
    Color Mode=::NewEnumerator2
    Lifetime=&RandomRangeFloat(Minimum=1, Maximum=1.3, Evaluation Type=::NewEnumerator1)
    Sprite Rotation Mode=::NewEnumerator2
    Sprite Size Max=<20 30>
    Sprite Size Min=<15 10>
    Sprite Size Mode=::NewEnumerator2
    Sprite UV Mode=::NewEnumerator1
    UsePositionOffset=false
    Write Material Random=true
   )
   &DynamicMaterialParameters(
    Index 0 Param 1=8
    Index 0 Param 2=&RandomRangeFloat(Minimum=0, Maximum=2, Evaluation Type=::NewEnumerator1)
    Module.Index0Float0Write=true
    Module.Index0Float1Write=true
    Module.Index0Float2Write=false
    Module.Index0Float3Write=false
   )
   &SystemLocation(MaskSpawnGroup=false, Offset=&RandomRangeVector(Minimum=<0 -10 80>, Maximum=<0 20 25>))
   &ShapeLocation(MaskSpawnGroup=false, Override Seed=false, Sphere Radius=2, UseOwnerScale=true)
  ]
  particleUpdateScript [
   &ParticleState()
   &ScaleSpriteSize(
    Non-Uniform Curve Sprite Scale=DataInterface() {propertyValues {
      XCurve {keys [{tangentMode=RCTM_User, time=0.1096, value=1.4555}, {tangentMode=RCTM_User, time=0.9934, value=2.5667}]}
      YCurve {keys [{tangentMode=RCTM_User, time=0.1096, value=1.5}, {tangentMode=RCTM_User, time=1, value=3}]}
     }}
    Scale Sprite Size Mode=::NewEnumerator3
   )
   &ScaleColor(
    Scale Alpha=&FloatFromCurve(FloatCurve=DataInterface() {propertyValues {Curve {keys [
         {arriveTangent=0.9953, time=0, value=0}
         {arriveTangent=-0.0889, interpMode=RCIM_Cubic, leaveTangent=-0.0889, tangentMode=RCTM_User, time=0.1488, value=0.9953}
         {arriveTangent=-0.0107, interpMode=RCIM_Cubic, leaveTangent=-2.5383, tangentMode=RCTM_Break, time=0.4308, value=0.99}
         <at 1 0>
        ]}}})
    Scale RGB=<0.5501 0.9597 1.19>
    ScaleA=true
    ScaleRGB=true
   )
   &DynamicMaterialParameters(
    Index 0 Param 1=8
    Index 0 Param 4=0.5
    Module.Index0Float0Write=false
    Module.Index0Float1Write=false
    Module.Index0Float2Write=true
    Module.Index0Float3Write=true
   ) {module=DynamicMaterialParameters001}
   &SubUVAnimation(
    End Frame Range Override=24
    "Sprite Renderer "=DataInterface() {propertyValues {SpriteRenderer=&self:MistIdle_1.NiagaraSpriteRendererProperties_0}}
    Start Frame Range Override=16
    SubUV Animation Mode:ENiagara_SubUVLookupModeV2=::NewEnumerator1
    SubUV Curve=DataInterface() {propertyValues {Curve {keys [
        {arriveTangent=19.1236, interpMode=RCIM_Cubic, leaveTangent=19.1236, tangentMode=RCTM_User, time=0, value=16}
        {arriveTangent=0.2086, interpMode=RCIM_Cubic, leaveTangent=0.2086, tangentMode=RCTM_User, time=1, value=25}
       ]}}}
    UseEndFrame=false
    UseStartFrame=false
   )
   &AccelerationForce(Acceleration=$Particles.Acceleration)
   &Drag(Use Linear Drag=true, Use Rotational Drag=true)
   &SolveForcesAndVelocity(Clamp Velocity=false, Limit Acceleration=false)
  ]
  rendererClasses [&NiagaraSpriteRendererProperties]
  renderers [{rendererClass=&NiagaraSpriteRendererProperties}]
 }
 MistIdle_Low {
  emitterUpdateScript [
   &EmitterState()
   &SpawnRate(
    SpawnRate=&MakeCustomFloatFromBool(
     Bool=&SetBoolByFloatComparison(A=&VectorLength(Vector=$Engine.Owner.Velocity), B=100, Comparison Type=::NewEnumerator6)
     True Float=8
     False Float=0
    )
    Use Spawn Probability=false
   )
  ]
  particleSpawnScript [
   /SetVariables(Particles.Acceleration:Vector3f=&RandomRangeVector(Minimum=<0 0 -20>, Maximum=<0 0 -25>, Evaluation Type=::NewEnumerator1)) {isSetParameters=true, module=SetVariables_#1}
   &InitializeParticle(
    Color Maximum=<rgba 1 1 1 0.02>
    Color Minimum=<rgba 2 2 2 0.02>
    Color Mode=::NewEnumerator2
    Lifetime=&RandomRangeFloat(Minimum=1, Maximum=1.3, Evaluation Type=::NewEnumerator1)
    Sprite Rotation Mode=::NewEnumerator2
    Sprite Size Max=<20 30>
    Sprite Size Min=<15 10>
    Sprite Size Mode=::NewEnumerator2
    Sprite UV Mode=::NewEnumerator1
    UsePositionOffset=false
    Write Material Random=true
   )
   &DynamicMaterialParameters(
    Index 0 Param 1=8
    Index 0 Param 2=&RandomRangeFloat(Minimum=0, Maximum=2, Evaluation Type=::NewEnumerator1)
    Module.Index0Float0Write=true
    Module.Index0Float1Write=true
    Module.Index0Float2Write=false
    Module.Index0Float3Write=false
   )
   &SystemLocation(MaskSpawnGroup=false, Offset=&RandomRangeVector(Minimum=<0 -10 80>, Maximum=<0 20 25>))
   &ShapeLocation(MaskSpawnGroup=false, Override Seed=false, Sphere Radius=2, UseOwnerScale=true)
  ]
  particleUpdateScript [
   &ParticleState()
   &ScaleSpriteSize(
    Non-Uniform Curve Sprite Scale=DataInterface() {propertyValues {
      XCurve {keys [{tangentMode=RCTM_User, time=0.1096, value=1.4555}, {tangentMode=RCTM_User, time=0.9934, value=2.5667}]}
      YCurve {keys [{tangentMode=RCTM_User, time=0.1096, value=1.5}, {tangentMode=RCTM_User, time=1, value=3}]}
     }}
    Scale Sprite Size Mode=::NewEnumerator3
   )
   &ScaleColor(
    Scale Alpha=&FloatFromCurve(FloatCurve=DataInterface() {propertyValues {Curve {keys [
         {arriveTangent=0.9953, time=0, value=0}
         {arriveTangent=-0.0889, interpMode=RCIM_Cubic, leaveTangent=-0.0889, tangentMode=RCTM_User, time=0.1488, value=0.9953}
         {arriveTangent=-0.0107, interpMode=RCIM_Cubic, leaveTangent=-2.5383, tangentMode=RCTM_Break, time=0.4308, value=0.99}
         <at 1 0>
        ]}}})
    Scale RGB=<0.5501 0.9597 1.19>
    ScaleA=true
    ScaleRGB=true
   )
   &DynamicMaterialParameters(
    Index 0 Param 1=8
    Index 0 Param 4=0.5
    Module.Index0Float0Write=true
    Module.Index0Float1Write=true
    Module.Index0Float2Write=true
    Module.Index0Float3Write=true
   ) {module=DynamicMaterialParameters001}
   &SubUVAnimation(
    End Frame Range Override=24
    "Sprite Renderer "=DataInterface() {propertyValues {SpriteRenderer=&self:MistIdle_Low_1.NiagaraSpriteRendererProperties_2}}
    Start Frame Range Override=16
    UseEndFrame=true
    UseStartFrame=true
   )
   &AccelerationForce(Acceleration=$Particles.Acceleration)
   &Drag(Use Linear Drag=true, Use Rotational Drag=true)
   &CameraOffset(Camera Offset Amount=-50)
   &SolveForcesAndVelocity(Clamp Velocity=false, Limit Acceleration=false)
  ]
  rendererClasses [&NiagaraSpriteRendererProperties]
  renderers [{rendererClass=&NiagaraSpriteRendererProperties}]
 }
 MistTrail {
  emitterUpdateScript [
   &EmitterState()
   &SpawnPerUnit(
    Max Movement Threshold=1000
    Module.Use Max Movement Threshold=true
    Module.Use Movement Tolerance=true
    Movement Tolerance=0.1
    Spawn Spacing=30
    Use Spawn Probability=false
   )
  ]
  particleSpawnScript [
   /SetVariables(Particles.Acceleration:Vector3f=&RandomRangeVector(Minimum=<0 0 -25>, Maximum=<0 0 -32>, Evaluation Type=::NewEnumerator1)) {isSetParameters=true}
   &InitializeParticle(
    Color Maximum=<rgba 0.5 0.5 0.5 0.15>
    Color Minimum=<rgba 1 1 1 0.15>
    Color Mode=::NewEnumerator2
    Lifetime=&RandomRangeFloat(Minimum=0.4, Maximum=0.6, Evaluation Type=::NewEnumerator1)
    Sprite Rotation Mode=::NewEnumerator2
    Sprite Size Max=<35 30>
    Sprite Size Min=<28 22>
    Sprite Size Mode=::NewEnumerator2
    UsePositionOffset=false
    Write Material Random=true
   )
   &DynamicMaterialParameters(
    Index 0 Param 1=8
    Index 0 Param 2=&RandomRangeFloat(Minimum=0, Maximum=2, Evaluation Type=::NewEnumerator1)
    Module.Index0Float0Write=true
    Module.Index0Float1Write=true
    Module.Index0Float2Write=false
    Module.Index0Float3Write=false
   )
   &SystemLocation(
    MaskSpawnGroup=false
    Offset=&RandomRangeVector(Minimum=<-8 -10 80>, Maximum=<-8 20 20>, Evaluation Type=::NewEnumerator1)
   )
   &ShapeLocation(MaskSpawnGroup=false, Override Seed=false, Sphere Radius=2, UseOwnerScale=true)
   &InheritVelocity(Inherited Velocity Amount Scale=<0.6 0.6 0.6>, Inherited Velocity Speed Limit=0, Speed Limit=false, UseSpeedThreshold=false)
  ]
  particleUpdateScript [
   &ParticleState()
   &ScaleSpriteSize(
    Non-Uniform Curve Sprite Scale=DataInterface() {propertyValues {
      XCurve {keys [{tangentMode=RCTM_User, time=0.1096, value=1.4555}, {tangentMode=RCTM_User, time=0.9934, value=2.5667}]}
      YCurve {keys [{tangentMode=RCTM_User, time=0.1096, value=1.5}, {tangentMode=RCTM_User, time=1, value=3}]}
     }}
    Scale Sprite Size Mode=::NewEnumerator3
   )
   &ScaleColor(
    Scale Alpha=&FloatFromCurve(FloatCurve=DataInterface() {propertyValues {Curve {keys [{arriveTangent=0.99, time=0, value=0}, {arriveTangent=-0.99, leaveTangent=-0.99, time=0.3013, value=0.99}, <at 1 0>]}}})
    Scale RGB=<0.5501 0.9597 1.19>
    ScaleA=true
    ScaleRGB=true
   )
   &DynamicMaterialParameters(
    Index 0 Param 4=0.5
    Module.Index0Float0Write=false
    Module.Index0Float1Write=false
    Module.Index0Float2Write=true
    Module.Index0Float3Write=true
   ) {module=DynamicMaterialParameters001}
   &SubUVAnimation(
    End Frame Range Override=24
    "Sprite Renderer "=DataInterface() {propertyValues {SpriteRenderer=&self:MistTrail_0.NiagaraSpriteRendererProperties_2}}
    Start Frame Range Override=14
    UseEndFrame=true
    UseStartFrame=true
   )
   &AccelerationForce(Acceleration=$Particles.Acceleration)
   &Drag(Use Linear Drag=true, Use Rotational Drag=true)
   &SolveForcesAndVelocity(Clamp Velocity=false, Limit Acceleration=false)
  ]
  rendererClasses [&NiagaraSpriteRendererProperties]
  renderers [{rendererClass=&NiagaraSpriteRendererProperties}]
 }
}
summary {
 systemName=NS_Pickaxe_IceCream_Idle
 emitters [
  {emitterName=MistTrail, rendererClasses [&NiagaraSpriteRendererProperties]}
  {emitterName=MistIdle, rendererClasses [&NiagaraSpriteRendererProperties]}
  {emitterName=MistIdle_Low, rendererClasses [&NiagaraSpriteRendererProperties]}
 ]
 userVariables []
}
