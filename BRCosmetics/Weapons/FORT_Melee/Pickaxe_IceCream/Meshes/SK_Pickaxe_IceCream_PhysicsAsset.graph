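# Physics asset for the SK_Pickaxe_IceCream skeletal mesh, an ice cream themed two-handed melee pickaxe cosmetic.
# It defines a single capsule body on the Melee_TwoHanded_Handle bone, offset along the handle to approximate the weapon with one simple shape.
# It has no constraints, so there is no jointed or secondary physics between parts.
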PhysicsAsset /BRCosmetics/Weapons/FORT_Melee/Pickaxe_IceCream/Meshes/SK_Pickaxe_IceCream_PhysicsAsset
bodies {Melee_TwoHanded_Handle [{shapeType=Capsule, length=102.058, radius=24.6957, center=<1.0983 -0.8153 51.3598>, rotation=<rot -0.5052 -90 -0.2499>}]}
constraints []
